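# Default Lit demo material for the Ancients statue props in the Asteria environment, flagged for use with Nanite.
# It holds a single sample of the Demo_Statues_Texture_D texture, likely the diffuse map by its suffix, and one scalar constant.
# No links from these two nodes to material outputs appear, and there are no parameters for instances to override.
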Material /Game/Environments/Asteria/Props/Ancients/Ancients_Statues_A/Demo/Materials/Demo_Ancients_Statue_Mat
graph {
 properties {bUsedWithNanite=true, pixelDepthOffsetMode=PDOM_Legacy, usedShadingModels=Default Lit}
 nodes [
  Constant() {r=0.6}
  TextureSample() {texture=/Game/Environments/Asteria/Props/Ancients/Ancients_Statues_A/Demo/Textures/Demo_Statues_Texture_D}
 ]
}
